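# MetaSound source preset that plays the foley audio for the Cobbler emote.
# It inherits its graph from MSS_Emote_Foley, feeding the Emote_Cobbler wave into the foley sounds input and leaving the loop sounds empty.
# It outputs stereo through the one-shot source interface, using the emote foley sound class, property sheet and concurrency rules with a raised volume.
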MetaSoundSource /BRCosmetics/Sounds/Emotes/Cobbler/MSS_Emote_Foley_Cobbler
preset_parent=/FortMetasound/Cosmetics/Emote/MSS_Emote_Foley
inputs [
 CopyrightedAudio {type=Bool, access=Constructor}
 DelayLoopByStart {type=Bool, access=Constructor}
 FadeInTime {type=Time, access=Constructor}
 Foley.Loop.Sounds {type=WaveAsset:Array, access=Constructor}
 Foley.Sounds {type=WaveAsset:Array, access=Constructor}
 IntroVolumeMultiplier {type=Float}
 Loop.DelayTime {type=Time}
 Loop.FadeCurveExp {type=Float}
 Loop.FadeTimeIn {type=Float}
 Looping {type=Bool, access=Constructor}
 LoopVolumeMultiplier {type=Float}
 PitchTemplateMultiplier {type=Float}
 UE.Source.OnPlay {type=Trigger}
]
interfaces [UE.OutputFormat.Stereo:1.0, UE.Source.OneShot:1.0, UE.Source:1.1]
outputs [
 "UE.OutputFormat.Stereo.Audio:0" {type=Audio}
 "UE.OutputFormat.Stereo.Audio:1" {type=Audio}
 UE.Source.OneShot.OnFinished {type=Trigger}
]
pages [Default {input_defaults {
   CopyrightedAudio=null
   DelayLoopByStart=null
   FadeInTime=0
   IntroVolumeMultiplier=1
   Loop.DelayTime=0
   Loop.FadeCurveExp=1
   Loop.FadeTimeIn=0
   LoopVolumeMultiplier=1
   Looping=null
   PitchTemplateMultiplier=1
   UE.Source.OnPlay=null
   Foley.Loop.Sounds []
   Foley.Sounds [&/Emote_Cobbler]
  }}]
properties {
 assetImportData=&self:AssetImportData
 audioPropertiesSheet=/BRCosmetics/Sounds/Settings/PropertySheets/Emotes/APS_Cosmetic_Emote_Foley
 numChannels=2
 outputFormat=Stereo
 sampleRate=48000
 soundClassObject=/Game/Sounds/FortSoundClassesAndModes/Fort_Foley_Emote
 volume=2
 concurrencySet [/Game/Sounds/Concurrency_Rules/Athena_EmoteFoley_Conc]
}
